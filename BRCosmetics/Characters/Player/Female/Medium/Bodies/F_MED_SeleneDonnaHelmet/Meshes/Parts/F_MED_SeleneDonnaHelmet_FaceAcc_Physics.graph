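# Physics asset for the face accessory mesh of the female SeleneDonnaHelmet cosmetic, covering two mirrored helmet bone chains.
# Each side chains four capsule bodies from a base capsule, which is fully angular-locked to a large head sphere.
# Chain joints set zero twist and swing 1 limits with small swing 2 angles, suggesting single-plane sway.
# The sides differ: the right chain's first joint is angular-locked, while the left's is not.
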PhysicsAsset /BRCosmetics/Characters/Player/Female/Medium/Bodies/F_MED_SeleneDonnaHelmet/Meshes/Parts/F_MED_SeleneDonnaHelmet_FaceAcc_Physics
bodies {
 base_helmet_01_l [{shapeType=Capsule, length=0.505, radius=0.505, shapeName=base_helmet_01_l_capsule, rotation=<rot 0 -90 0>}]
 base_helmet_01_r [{shapeType=Capsule, length=0.505, radius=0.505, shapeName=base_helmet_01_r_capsule, rotation=<rot 0 -90 0>}]
 dyn_helmet_01_l [{shapeType=Capsule, length=1.5, radius=1, shapeName=dyn_helmet_01_l_capsule, center=<0.5 0 0>}]
 dyn_helmet_01_r [{shapeType=Capsule, length=1.5, radius=1, shapeName=dyn_helmet_01_r_capsule, center=<0.5 0 0>}]
 dyn_helmet_02_l [{shapeType=Capsule, length=1.5, radius=1, shapeName=dyn_helmet_02_l_capsule, center=<0.5 0 0>}]
 dyn_helmet_02_r [{shapeType=Capsule, length=1.5, radius=1, shapeName=dyn_helmet_02_r_capsule, center=<1 0 0>}]
 dyn_helmet_03_l [{shapeType=Capsule, length=1.5, radius=1, shapeName=dyn_helmet_03_l_capsule, center=<0.5 0 0>}]
 dyn_helmet_03_r [{shapeType=Capsule, length=1.5, radius=1, shapeName=dyn_helmet_03_r_capsule, center=<0.5 0 0>}]
 dyn_helmet_04_l [{shapeType=Capsule, length=1.5, radius=1, shapeName=dyn_helmet_04_l_capsule, center=<0.5 0 0>}]
 dyn_helmet_04_r [{shapeType=Capsule, length=1.5, radius=1, shapeName=dyn_helmet_04_r_capsule, center=<0.5 0 0>}]
 head [{shapeType=Sphere, radius=13.5274, shapeName=head_sphere, center=<4.8653 0.934 -0.0104>}]
}
constraints [
 {bone1Name=dyn_helmet_04_r, bone2Name=dyn_helmet_03_r, swing1LimitDegrees=0, swing2LimitDegrees=5, twistLimitDegrees=0}
 {bone1Name=dyn_helmet_03_r, bone2Name=dyn_helmet_02_r, swing1LimitDegrees=0, swing2LimitDegrees=5, twistLimitDegrees=0}
 {bone1Name=dyn_helmet_02_r, bone2Name=dyn_helmet_01_r, swing1LimitDegrees=0, swing2LimitDegrees=5, twistLimitDegrees=0}
 {
  bone1Name=dyn_helmet_01_r
  bone2Name=base_helmet_01_r
  swing1LimitDegrees=0
  swing1Motion=Locked
  swing2LimitDegrees=10
  swing2Motion=Locked
  twistLimitDegrees=0
  twistMotion=Locked
 }
 {
  bone1Name=base_helmet_01_r
  bone2Name=head
  swing1LimitDegrees=45
  swing1Motion=Locked
  swing2LimitDegrees=45
  swing2Motion=Locked
  twistLimitDegrees=45
  twistMotion=Locked
 }
 {
  bone1Name=base_helmet_01_l
  bone2Name=head
  swing1LimitDegrees=45
  swing1Motion=Locked
  swing2LimitDegrees=45
  swing2Motion=Locked
  twistLimitDegrees=45
  twistMotion=Locked
 }
 {bone1Name=dyn_helmet_01_l, bone2Name=base_helmet_01_l, swing1LimitDegrees=0, swing2LimitDegrees=25, twistLimitDegrees=0}
 {bone1Name=dyn_helmet_02_l, bone2Name=dyn_helmet_01_l, swing1LimitDegrees=0, swing2LimitDegrees=4, twistLimitDegrees=0}
 {bone1Name=dyn_helmet_03_l, bone2Name=dyn_helmet_02_l, swing1LimitDegrees=0, swing2LimitDegrees=5, twistLimitDegrees=0}
 {bone1Name=dyn_helmet_04_l, bone2Name=dyn_helmet_03_l, swing1LimitDegrees=0, swing2LimitDegrees=3, twistLimitDegrees=0}
]
